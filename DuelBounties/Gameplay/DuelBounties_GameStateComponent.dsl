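; Game state component that runs a server-side, single-slot matchmaking queue for player duel bounties.
; The first player waits on a randomized scalable-float timer; a second joiner becomes their target, and elimination or EndPlay triggers queue cleanup.
; On timeout, it seeks a tag-filtered nearest enemy "Phoebe" and grants it duel GameplayEffects.
; Pairing broadcasts quest-advance verbs, then after a short delay sends TargetFound gameplay events to both squads.
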(event Custom|TryAddPlayerToDuelQueue
  (params (InstigatingPlayerController :type Fort%20Player%20Controller%20Object%20Reference) (DuelQuestDefinition :type Fort%20Quest%20Item%20Definition%20Athena%20Object%20Reference) (PhoebeGameplayEffectClass :type Gameplay%20Effect%20Class%20Reference))
  (bind _instigating_player_controller (Variables|Default|GetInstigating%20Player%20Controller))
  ;[0 0 0] Tries to add a player controller to the duel waiting queue. If a player is already waiting, set them as targets
  (Utilities|SwitchHasAuthority
    (:Authority
      (Utilities|IsValid InstigatingPlayerController
        (:Is%20Valid
          (Utilities|IsValid DuelQuestDefinition
            (:Is%20Valid
              (Variables|Default|SetDuel%20Quest%20Definition DuelQuestDefinition)
              (if (Utilities|IsValidClass PhoebeGameplayEffectClass)
                (Variables|Default|SetPhoebe%20Gameplay%20Effect%20Class PhoebeGameplayEffectClass)
                (Utilities|IsValid _instigating_player_controller
                  (:Is%20Valid
                    (CallFunction|SetDuelistTargets (Game|Player|GetPlayerPawn _instigating_player_controller) (Game|Player|GetPlayerPawn InstigatingPlayerController)))
                  (:Is%20Not%20Valid
                    (bind _output_get (Variables|Default|SetInstigating%20Player%20Controller InstigatingPlayerController))
                    ;;[0.15 0.15 0.15] Bind an event to track if this player is eliminated before another player accepts
                    (bind _returnvalue (Fortnite|Utilities|AttemptGetPlayerStateforPawn (Game|Player|GetPlayerPawn _output_get)))
                    (EventDispatchers|AssignOnPawnDiedDelegate _returnvalue (EventDispatchers|CreateEvent :!event Custom|OnQueuedDuelistEliminated))
                    ;;[0.15 0.15 0.15] Start a timer to allow another player to enter the queue
                    (bind _returnvalue_2 (Utilities|Time|SetTimerbyEvent (EventDispatchers|CreateEvent :!event Custom|QueueTimeUp) (Math|Random|RandomFloatinRange (ScalableFloat|GetValueatLevel (Variables|Default|GetQueueWaitTimeMin) 1.0) (ScalableFloat|GetValueatLevel (Variables|Default|GetQueueWaitTimeMax) 1.0))))
                    (Variables|Default|SetQueueTimerHandle _returnvalue_2)))))) ; Is quest def provided valid?
        )) ; Is player controller valid?
    )))

(event Custom|SetDuelistTargets
  (params (FirstPlayer :type Fort%20Player%20Pawn%20Object%20Reference) (SecondPlayer :type Fort%20Player%20Pawn%20Object%20Reference))
  ;[0 0 0] Set the duelists in the queue to be targets of each other
  (Variables|Default|SetFirst%20Player FirstPlayer)
  (Variables|Default|SetSecond%20Player SecondPlayer)
  (CallFunction|AdvanceQueueQuestforDuelists (Variables|Default|GetFirst%20Player) (Variables|Default|GetSecond%20Player))
  (bind _returnvalue (Utilities|Time|SetTimerbyEvent (delegate Custom|DelaySendGameplayEvent) 2.5)) ; Wait for quest to advance and clear from the UI before sending an event to trigger associated abilities and logic for the duel starting
  (Variables|Default|SetDelayTimer _returnvalue))

(event Custom|DelaySendGameplayEvent
  ;[0 0 0] Set the duelists in the queue to be targets of each other
  (Utilities|FlowControl|Sequence
    (:then_0
      ;; Set the second player as a target of the first player and their squad
      (bind _squadmatesreturned (CallFunction|GetSquadmatesforGivenPlayer (Variables|Default|GetFirst%20Player)))
      (for _array_element _squadmatesreturned
        (Ability|SendGameplayEventtoActor _array_element "(TagName=\"Event.PlayerDuel.TargetFound\")" (Utilities|Struct|MakeGameplayEventData "(TagName=\"\")" (Variables|Default|GetFirst%20Player) (Variables|Default|GetSecond%20Player) (Variables|Default|GetDuel%20Quest%20Definition))) ; Set target of first player
      ))
    (:then_1
      ;; Set the first player as a target of the second player and their squad
      (bind _squadmatesreturned_2 (CallFunction|GetSquadmatesforGivenPlayer (Variables|Default|GetSecond%20Player)))
      (for _array_element_2 _squadmatesreturned_2
        (Ability|SendGameplayEventtoActor _array_element_2 "(TagName=\"Event.PlayerDuel.TargetFound\")" (Utilities|Struct|MakeGameplayEventData "(TagName=\"\")" (Variables|Default|GetSecond%20Player) (Variables|Default|GetFirst%20Player) (Variables|Default|GetDuel%20Quest%20Definition)))))
    (:then_2
      (CallFunction|ClearUsedVariables))))

(event Custom|AdvanceQueueQuestForDuelists
  (params (FirstDuelist :type Fort%20Player%20Pawn%20Object%20Reference) (SecondDuelist :type Fort%20Player%20Pawn%20Object%20Reference))
  ;[0 0 0] Advance hidden queue quest for duelists
  (for _array_element (Utilities|Array|MakeArray FirstDuelist SecondDuelist)
    (CustomVerb|BroadcastCustomVerbMessage (CustomVerb|MakeCustomVerbMessage _array_element "(TagName=\"CustomVerb.PlayerDuel.TargetFound\")"))))

(event Custom|QueueTimeUp
  (bind _nearestphoebe (Variables|Default|GetNearestPhoebe))
  (bind _returnvalue (Ability|GetFortAbilitySystemComponent _nearestphoebe))
  ;[0 0 0] When queue timer is up, find eligble phoebe who matches filter criteria to select as the target or trigger a fail state if another one can't be found
  (bind _targetplayerpawn (CallFunction|GetNearestEnemyPlayer (Variables|Default|GetInstigating%20Player%20Controller)))
  (Utilities|IsValid _targetplayerpawn
    (:Is%20Valid
      (Variables|Default|SetNearestPhoebe _targetplayerpawn)
      (Utilities|IsValid _returnvalue
        (:Is%20Valid
          (GameplayEffects|ApplyGameplayEffectToSelf :self _returnvalue :GameplayEffectClass "/JobBoard/Blueprints/Shared/GE_Granted_JobBoardBriefing_Instigator.GE_Granted_JobBoardBriefing_Instigator_C" :Level 0.0) ; Apply the instigator GE to phoebe
          (Utilities|FlowControl|DelayUntilNextTick)
          (GameplayEffects|ApplyGameplayEffectToSelf :self (Ability|GetFortAbilitySystemComponent (Variables|Default|GetNearestPhoebe)) :GameplayEffectClass (Variables|Default|GetPhoebe%20Gameplay%20Effect%20Class) :Level 0.0) ; Apply phoebe's version of the duel ability for tracking
          (CallFunction|SetDuelistTargets (Game|Player|GetPlayerPawn (Variables|Default|GetInstigating%20Player%20Controller)) (Variables|Default|GetNearestPhoebe)))
        (:Is%20Not%20Valid
          _)))
    (:Is%20Not%20Valid
      (Development|PrintString "No valid players that meet filter criteria available for duel!" true true "(R=1.000000,G=0.000000,B=0.600193,A=1.000000)" 5.0)))
  (CustomVerb|BroadcastCustomVerbMessage (CustomVerb|MakeCustomVerbMessage (Variables|Default|GetInstigating%20Player%20Controller) "(TagName=\"CustomVerb.JobBoard.Failed\")")))

(event Custom|ClearUsedVariables
  (bind _nullplayerpawn (Variables|Default|GetNullPlayerPawn))
  (bind _delaytimer (Variables|Default|GetDelayTimer))
  (bind _queuetimerhandle (Variables|Default|GetQueueTimerHandle))
  ;[0 0 0] Clear out variables so they can be reused for a new player entering the queue
  (Variables|Default|SetInstigating%20Player%20Controller (Variables|Default|GetNullPlayerController))
  (Variables|Default|SetFirst%20Player _nullplayerpawn)
  (Variables|Default|SetSecond%20Player _nullplayerpawn)
  (if (Utilities|Time|IsValidTimerHandle _queuetimerhandle)
    (Utilities|Time|ClearandInvalidateTimerbyHandle _queuetimerhandle)
    (if (Utilities|Time|IsValidTimerHandle _delaytimer)
      (Utilities|Time|ClearandInvalidateTimerbyHandle _delaytimer))))

(event Custom|OnQueuedDuelistEliminated
  (params (PawnDamageInfo :type Pawn%20Damage%20Data%20Structure%20%28by%20ref%29))
  (bind _damagedactor (Utilities|Struct|BreakPawnDamageData PawnDamageInfo))
  ;[0 0 0] If a waiting player is eliminated, cancel the queue timer and null out references so a new set of players can be selected
  (if (and (Utilities|IsValid (Fortnite|GetFortPlayerControllerfromActor _damagedactor) :!class /Script/Engine.KismetSystemLibrary) (== (Fortnite|GetFortPlayerControllerfromActor _damagedactor) (Variables|Default|GetInstigating%20Player%20Controller)))
    (CallFunction|ClearUsedVariables)
    (CustomVerb|BroadcastCustomVerbMessage (CustomVerb|MakeCustomVerbMessage _damagedactor "(TagName=\"CustomVerb.JobBoard.Failed\")"))))

(event EventEndPlay
  (params (EndPlayReason :type EEndPlayReason%20Enum :default "Destroyed"))
  (bind _instigating_player_controller (Variables|Default|GetInstigating%20Player%20Controller))
  ;[0 0 0] End play - unbind any events and clear out the timer handle
  (|CallParentFunction EndPlayReason)
  (Utilities|IsValid _instigating_player_controller
    (:Is%20Valid
      (bind _returnvalue (Fortnite|Utilities|AttemptGetPlayerStateforPawn (Game|Player|GetPlayerPawn _instigating_player_controller)))
      (EventDispatchers|UnbindEventfromOnPawnDiedDelegate _returnvalue (EventDispatchers|CreateEvent :!event Custom|OnQueuedDuelistEliminated))
      (CallFunction|ClearUsedVariables))))

(fn GetNearestEnemyPlayer
  (params (InstigatingPlayerController :type Fort%20Player%20Controller%20Object%20Reference))
  ;[0 0 0] Find the nearest phoebe to assign to the queue. May return null
  (Utilities|SwitchHasAuthority
    (:Authority
      (Utilities|IsValid InstigatingPlayerController
        (:Is%20Valid
          (bind _asfort_game_state_component_athena_quests (Utilities|Casting|CastToFortGameStateComponentAthena_Quests (Actor|GetComponentbyClass (Fortnite|Athena|GameState|GetGameStateAthena) (LocalVariables|GetAthenaQuestsComponent))))
          (bind _returnvalue (Class|FortGameStateComponentAthenaQuests|FindClosestEnemyPlayer _asfort_game_state_component_athena_quests (Game|Player|GetPlayerPawn InstigatingPlayerController) 5 (LocalVariables|GetTargetMustHaveTags)))
          (Utilities|IsValid _returnvalue
            (:Is%20Valid
              (bind _asfort_player_pawn (Utilities|Casting|CastToFortPlayerPawn (PlayerState|GetPawn _returnvalue)))
              (return _asfort_player_pawn))))))))

(fn GetSquadmatesForGivenPlayer
  (params (FortPlayerPawn :type Fort%20Player%20Pawn%20Object%20Reference))
  ;[0 0 0] Get the squadmates of the given player
  (Utilities|IsValid FortPlayerPawn
    (:Is%20Valid
      (Utilities|Array|Clear (LocalVariables|GetSquadmates))
      (bind _returnvalue (Fortnite|Utilities|AttemptGetPlayerStateforPawn FortPlayerPawn))
      (bind _asfort_player_state_athena (Utilities|Casting|CastToFortPlayerStateAthena _returnvalue))
      (bind _returnvalue_2 (Class|FortGameStateAthena|GetSquadMemberPlayerStates (Fortnite|Athena|GameState|GetGameStateAthenafromContext FortPlayerPawn) _asfort_player_state_athena true))
      (for _array_element _returnvalue_2
        (bind _asfort_player_pawn (Utilities|Casting|CastToFortPlayerPawn (PlayerState|GetPawn _array_element)))
        (Utilities|Array|AddUnique (LocalVariables|GetSquadmates) _asfort_player_pawn))
      (return (LocalVariables|GetSquadmates)))))
