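; Duel Bounties queue object that collects player controllers server-side and matches them as each other's duel targets.
; Once two or more are queued, delayed steps broadcast a quest-advancing custom verb to all, then send the first two a TargetFound gameplay event naming the other.
; The whole queue is then cleared.
; A client-only event shows the locally controlled instigator a waiting message, with a splitscreen-aware check.
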(event Custom|AddPlayerToDuelQueue
  (params (InstigatingPlayerController :type Fort%20Player%20Controller%20Object%20Reference))
  ;[0 0 0] Add player to duel queue, when called
  (Networking|SwitchHasAuthority
    (:Authority
      (Utilities|Array|AddUnique (Variables|Default|GetDuelQueue) InstigatingPlayerController)
      (if (>= (Utilities|Array|Length (Variables|Default|GetDuelQueue)) 2)
        (Utilities|FlowControl|Delay 0.5)
        (CallFunction|SetDuelistTargets)))))

(event Custom|AdvanceQuestForDuelists
  ;[0 0 0] Advance hidden queue quest for duelists
  (for _array_element (Variables|Default|GetDuelQueue)
    (CustomVerb|BroadcastCustomVerbMessage (CustomVerb|MakeCustomVerbMessage _array_element "(TagName=\"CustomVerb.PlayerDuel.TargetFound\")"))))

(event Custom|SetDuelistTargets
  ;[0 0 0] Set the duelists in the queue to be targets of each other
  (CallFunction|AdvanceQuestforDuelists)
  (Utilities|FlowControl|Delay 1.5)
  (for (_array_element _array_index) (Variables|Default|GetDuelQueue)
    (switch int _array_index
      (:0
        (bind _validcontrolleratindex (CallFunction|GetValidControllerforQueueIndex 1))
        (Ability|SendGameplayEventtoActor _array_element "(TagName=\"Event.PlayerDuel.TargetFound\")" (Utilities|Struct|MakeGameplayEventData "(TagName=\"\")" 0 _validcontrolleratindex)))
      (:1
        (bind _validcontrolleratindex_2 (CallFunction|GetValidControllerforQueueIndex 0))
        (Ability|SendGameplayEventtoActor _array_element "(TagName=\"Event.PlayerDuel.TargetFound\")" (Utilities|Struct|MakeGameplayEventData "(TagName=\"\")" 0 _validcontrolleratindex_2)))))
  (Utilities|Array|Clear (Variables|Default|GetDuelQueue)))

(event Custom|ShowQueueMessage
  (params (InstigatingPlayerController :type Fort%20Player%20Controller%20Object%20Reference))
  ;[0 0 0] Client only - Show a message for the player waiting in queue
  (Networking|SwitchHasAuthority
    (:Remote
      (if (CallFunction|IsInstigatorLocallyControlled InstigatingPlayerController)
        (Class|AthenaHUDContext|ShowUrgentQuestMessage (BlueprintContext|GetAthenaHUDContext) (Utilities|Struct|MakeUrgentQuestData "(TagName=\"SpecialEvent.PlayerDuel.Waiting\")" "(GameplayTags=)" 0.0 true false false false false true "" "NSLOCTEXT(\"[F0583E8331D309AD6782E0715686326C]\", \"6DC9987244E0F1A931AF979965277B91\", \"Duel Accepted\")" "NSLOCTEXT(\"[F0583E8331D309AD6782E0715686326C]\", \"95E27A17400AA90945FF5B802DAB37C7\", \"Waiting for a duelist...\")"))))))

(fn GetValidControllerForQueueIndex
  (params (Index :type Integer))
  (bind _duelqueue (Variables|Default|GetDuelQueue))
  (bind _output (Utilities|Array|Get(acopy) _duelqueue Index))
  (Utilities|IsValid _output
    (:Is%20Valid
      (return _output))))

(fn IsInstigatorLocallyControlled
  (params (Instigator :type Fort%20Player%20Controller%20Object%20Reference))
  ;[0 0 0] Returns true on the executing client if the instigator controller is locally controlled. Returns false if running on the server
  (if (Networking|IsServer)
    (return false)
    (else
      (Utilities|IsValid Instigator
        (:Is%20Valid
          (if (PlayerController|IsinSplitscreen Instigator)
            (for _array_element (Fortnite|GetLocalFortPlayerControllers true)
              (if (== _array_element Instigator)
                (return true)))
            (return false)
            (else
              (return true))))))))

(fn ConstructionScript)
