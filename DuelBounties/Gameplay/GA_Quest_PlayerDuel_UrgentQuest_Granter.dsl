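; Player-duel bounty ability that grants its owner a configured urgent quest through the transient quests component.
; Binds ended, expired and failed events, filtered by quest tag; expiry or failure sets a flag.
; Completed or claimed state passes that flag to an empty overridable hook, broadcasts a custom verb updating the granter quest, and ends the ability.
; Ability end resets the flag and unbinds ended and expired, but not failed.
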(event Ability|EventFailedToActivatePassiveAbility
  (params (ActorInfo :type Gameplay%20Ability%20Actor%20Info%20Structure))
  (|CallParentFunction ActorInfo))

(event Ability|EventActivateAbility
  ;[0 0 0] Activate ability - grant the urgent quest when event fires
  (|CallParentFunction)
  (Variables|Default|SetOwningPlayerPawn (CallFunction|GetAvatarasFortPlayerPawn))
  (CallFunction|GrantUrgentQuest))

(event Custom|GrantUrgentQuest
  (bind _owningplayerpawn (Variables|Default|GetOwningPlayerPawn))
  (bind _returnvalue (Class|FortPlayerPawn|GetTransientQuestsComponent _owningplayerpawn))
  ;[0 0 0] Grant the urgent quest as defined in class defaults and then bind event to when the urgent quest ends
  (Class|FortControllerComponentTransientQuests|GrantTransientQuest _returnvalue (Variables|UrgentQuestVariables|GetUrgent%20Quest%20to%20Grant) 0)
  (EventDispatchers|BindEventtoOnUrgentQuestEnded _returnvalue (EventDispatchers|CreateEvent :!event Custom|OnUrgentQuestEnded))
  (EventDispatchers|BindEventtoOnUrgentQuestTimeExpired _returnvalue (EventDispatchers|CreateEvent :!event Custom|OnUrgentQuestExpired))
  (EventDispatchers|AssignOnUrgentQuestFailed _returnvalue (EventDispatchers|CreateEvent :!event Custom|OnUrgentQuestFailed)))

(event Ability|EventOnEndAbility
  (params (bWasCancelled :type Boolean))
  (bind _returnvalue (CallFunction|GetAvatarasFortPlayerPawn))
  (bind _returnvalue_2 (Class|FortPlayerPawn|GetTransientQuestsComponent _returnvalue))
  ;[0 0 0] On End ability remove binds
  (|CallParentFunction bWasCancelled)
  (Utilities|IsValid _returnvalue
    (:Is%20Valid
      (Utilities|IsValid _returnvalue_2
        (:Is%20Valid
          (EventDispatchers|UnbindEventfromOnUrgentQuestEnded _returnvalue_2 (EventDispatchers|CreateEvent :!event Custom|OnUrgentQuestEnded))
          (EventDispatchers|UnbindEventfromOnUrgentQuestTimeExpired _returnvalue_2 (EventDispatchers|CreateEvent :!event Custom|OnUrgentQuestExpired))
          (Variables|Default|SetbQuestTimedOutOrFailed false))))))

(event Custom|OnUrgentQuestEnded
  (params (PlayerStateAthena :type Fort%20Player%20State%20Athena%20Object%20Reference) (UrgentQuestTag :type Gameplay%20Tag%20Structure%20%28by%20ref%29) (QuestState :type EFortQuestState%20Enum :default "Inactive") (CompletionPathContext :type Gameplay%20Tag%20Container%20Structure%20%28by%20ref%29))
  ;[0 0 0] When urgent quest ends, check if it is the current urgent quest; then send quest update for the granter to complete it
  (if (== (Variables|UrgentQuestVariables|GetUrgent%20Quest%20Tag) UrgentQuestTag)
    (if (or (Utilities|Enum|Equal(Enum) QuestState "Completed") (Utilities|Enum|Equal(Enum) QuestState "Claimed"))
      (CallFunction|AdditionalUrgentQuestEndedLogic (Variables|Default|GetbQuestTimedOutOrFailed))
      (CustomVerb|BroadcastCustomVerbMessage (CustomVerb|MakeCustomVerbMessage (Variables|Default|GetOwningPlayerPawn) (Variables|UrgentQuestVariables|GetCustomVerb)))
      (Ability|EndAbility))))

(event Custom|OnUrgentQuestExpired
  (params (UrgentQuestData :type Urgent%20Quest%20Data%20Structure%20%28by%20ref%29))
  ;[0 0 0] Additional logic - override in child blueprints for custom behavior
  (if (== (Variables|UrgentQuestVariables|GetUrgent%20Quest%20Tag) (Utilities|Struct|BreakUrgentQuestData UrgentQuestData))
    (Variables|Default|SetbQuestTimedOutOrFailed true)))

(event Custom|AdditionalUrgentQuestEndedLogic
  (params (bTimedOutOrFailed :type Boolean)))

(event Custom|OnUrgentQuestFailed
  (params (PlayerStateAthena :type Fort%20Player%20State%20Athena%20Object%20Reference) (UrgentQuestData :type Urgent%20Quest%20Data%20Structure%20%28by%20ref%29))
  ;[0 0 0] Additional logic - override in child blueprints for custom behavior
  (if (== (Variables|UrgentQuestVariables|GetUrgent%20Quest%20Tag) (Utilities|Struct|BreakUrgentQuestData UrgentQuestData))
    (Variables|Default|SetbQuestTimedOutOrFailed true)))
